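# IntroRotatingText is a UserWidget in the Apollo playlist folder that cycles a single CommonTextBlock through a list of strings, empty by default.
# RotateText shows the entry at TextsIndex, then increments the index and wraps it by array length, even when the array is empty.
# Construct calls RotateText immediately, and Tick feeds a Delay of TimeBetweenRotations, using its ignore-while-pending behavior as a repeating timer instead of a timer handle.
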WidgetBlueprint /Game/Athena/Playlists/Apollo/IntroRotatingText
graphs [EventGraph {nodes [
   RotateText() {is=title K2Node_CustomEvent}
   n1 = Get Texts() -> (Texts:Array of Texts)
   > n2 = Branch(Condition="integer > integer"(A=Length(TargetArray:Array of Texts=@n1), B=0) {is=title K2Node_CallFunction})
   > n3 = "SetText (Text)"(
    self=Get RotatingText() -> (RotatingText:Common Text Object Reference)
    InText="Get (a ref)"(Array:Array of Texts=Get Texts(), Dimension 1=Get TextsIndex() -> (TextsIndex:int=0)) -> (Output:"Text (by ref)")
   )
   n4 = Increment Int(" "=[@n3.then, @n2.else], Value=Get TextsIndex() -> (TextsIndex=0))
   n5 = Get Texts()
   > Set TextsIndex(TextsIndex:int="% (Integer)"(A=@n4, B=Length(TargetArray=@n5)) -> (ReturnValue=0)) -> (Output_Get:int=0)
   Event Tick()
   > Delay(Duration=Get TimeBetweenRotations() -> (TimeBetweenRotations:double))
   > RotateText() {is=title K2Node_CallFunction}
   Event Construct()
   > RotateText()
  ]}]
variables [
 Texts {type="array<string>", default []}
 TextsIndex {type=integer}
 TimeBetweenRotations {type=number, default=5}
]
widgets {
 info {inheritedWidgetCount=0, parentClass=&UserWidget, rootWidgetClass=&CommonTextBlock, widgetCount=1}
 widgets [RotatingText {
   widgetClassPath=&CommonTextBlock
   bIsVariable=true
   widget=RotatingText
   properties {
    text="{VeryVeryVeryVeryVeryVeryVeryVeryVeryVeryVeryVeryVeryVeryVeryVeryVeryVeryVeryVeryVeryVeryVeryLongText}"
    font {
     bForceMonospaced=false
     bMaterialIsStencil=false
     fontMaterial=&/M_UI_FlareNoise_Text
     fontObject=&BurbankSmall
     letterSpacing=0
     monospacedWidth=1
     size=20
     skewAmount=0
     typefaceFontName=Black
     outlineSettings {
      bApplyOutlineToDropShadows=false
      bMiteredCorners=false
      bSeparateFillAlpha=false
      outlineMaterial=null
      outlineSize=0
      outlineColor=<rgba 0 0 0 1>
     }
    }
    shadowOffset=<0 0>
   }
  }]
}
